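; Music battle instance for team Quickplay (Mainstage) in Pilgrim, managing player abilities and supplying song and analytics data.
; On authority, BeginPlay activates tagged battle abilities on each player's pawn and forwards playspace user removal to PlayerDisconnected.
; A flag-guarded cleanup ends those abilities on battle finish or EndPlay.
; Query functions read the current setlist song and index from band state, and difficulty from each player's state component.
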(event EventBeginPlay
  (|CallParentFunction)
  (CallFunction|ApplyGameplayAbilities)
  (Creative|Networking|Authority/Cosmetic
    (:Authority
      (Playspace|BindEventtoNotifyPlayspaceUserRemoved (Class|SparksMusicPlayspace|GetSparksMusicPlayspaceOfClass "/Script/PilgrimQuickplayRuntime.PilgrimQuickplayPlayspace") (delegate Custom|OnUserRemovedFromPlayspace)))))

(event Custom|OnUserRemovedFromPlayspace
  (params (RemovedUser :type Playspace%20User%20Structure%20%28by%20ref%29))
  (bind (_userid _leafplayspace _playerstatecached _controllercached) (Utilities|Struct|BreakPlayspaceUser RemovedUser))
  (CallFunction|PlayerDisconnected self (Actor|GetComponentbyClass _controllercached "/Script/PilgrimCoreRuntime.PilgrimPlayerControllerComponent")))

(event Collision|EventActorBeginOverlap
  (params (OtherActor :type Actor%20Object%20Reference)))

(event EventTick
  (params (DeltaSeconds :type Float%20%28single-precision%29)))

(event EventEndPlay
  (params (EndPlayReason :type EEndPlayReason%20Enum :default "Destroyed"))
  (|CallParentFunction EndPlayReason)
  (CallFunction|CleanUpGameplayAbilities "CompletedNormally"))

(fn ApplyGameplayAbilities
  (Networking|SwitchHasAuthority
    (:Authority
      (Utilities|FlowControl|Sequence
        (:then_0
          (for _array_element (Variables|PilgrimMusicBattleInstance|GetMusicBattlePlayers)
            (Abilities|TryActivateAbilitiesbyTag (Ability|GetAbilitySystemComponent (PlayerState|GetCurrentPawn _array_element)) (GameplayTags|MakeGameplayTagContainerfromTag (Variables|Default|GetMusicBattleAbilityTag)) true))
          (EventDispatchers|BindEventtoOnMusicBattleFinished (EventDispatchers|CreateEvent :!event CleanUpGameplayAbilities)))
        (:then_1
          (Variables|Default|SetAbilitiesInitialized true))))))

(fn CleanUpGameplayAbilities
  (params (EndReason :type EMusicBattleEndReason%20Enum :default "CompletedNormally"))
  (Networking|SwitchHasAuthority
    (:Authority
      ;[1 1 1] Only clean them up if they've been initialized and not already cleaned up
      (if (Variables|Default|GetAbilitiesInitialized)
        (Utilities|FlowControl|Sequence
          (:then_0
            (for _array_element (Variables|PilgrimMusicBattleInstance|GetMusicBattlePlayers)
              (Utilities|IsValid (PlayerState|GetCurrentPawn _array_element)
                (:Is%20Valid
                  (bind _outabilityhandles (GameplayAbilities|FindAllAbilitieswithTags (Ability|GetAbilitySystemComponent (PlayerState|GetCurrentPawn _array_element)) "(GameplayTags=((TagName=\"Mojo.Pilgrim.Abilities.MusicBattle\")))" true))
                  (for _array_element_2 _outabilityhandles
                    (Utilities|IsValid (.ReturnValue (Ability|GameplayAbility|GetGameplayAbilityfromSpecHandle (Ability|GetAbilitySystemComponent (PlayerState|GetCurrentPawn _array_element)) _array_element_2))
                      (:Is%20Valid
                        (Ability|EndAbility (.ReturnValue (Ability|GameplayAbility|GetGameplayAbilityfromSpecHandle (Ability|GetAbilitySystemComponent (PlayerState|GetCurrentPawn _array_element)) _array_element_2))))))))))
          (:then_1
            (Variables|Default|SetAbilitiesInitialized false)))))))

(fn GetNextSongToPlay
  ;[1 0.8887 0.6958] Return the current song from the chosen setlist
  (return (Class|PilgrimCoreBandStateComponent|GetCurrentSong (Class|PilgrimQuickplayBandStateComponent|GetBandState self))))

(fn GetPlacementAnalyticsData
  (bind _returnvalue (Class|SparksMusicPlayspace|GetSparksMusicPlayspaceOfClass "/Script/PilgrimCoreRuntime.PilgrimPlayspace"))
  (bind _returnvalue_2 (Actor|GetComponentbyClass _returnvalue "/Script/PilgrimQuickplayRuntime.PilgrimQuickplayBandStateComponent"))
  (Utilities|IsValid _returnvalue
    (:Is%20Valid
      (Utilities|IsValid _returnvalue_2
        (:Is%20Valid
          ; In Mainstage we don't record placement details, but round number is the current song number in the setlist
          (return (Utilities|Struct|MakePilgrimPlacementAnalyticsData -1 false -1 (Class|PilgrimQuickplayBandStateComponent|GetCurrentSetlistIndex _returnvalue_2))))
        (:Is%20Not%20Valid
          _)))
    (:Is%20Not%20Valid
      _))
  (Fortnite|FortShippingLog self "Unable to retrieve custom setlist" true)
  (return (LocalVariables|GetInvalidPlacementData)))

(fn GetSongDifficultyForPlayer
  (params (Player :type Fort%20Player%20State%20Athena%20Object%20Reference))
  (bind _returnvalue (Actor|GetComponentbyClass Player "/Script/PilgrimQuickplayRuntime.PilgrimQuickplayPlayerStateComponent"))
  (Utilities|IsValid _returnvalue
    (:Is%20Valid
      (return (Class|PilgrimCorePlayerStateComponent|GetSongDifficulty _returnvalue)))
    (:Is%20Not%20Valid
      (Development|PrintString "Unable to find song difficulty on player")
      (return "None"))))

(fn ConstructionScript
  (|CallParentFunction))
